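# Sound cue for the looping water wheel spot ambience at the Mill POI, routed to the outdoor ambient sound class.
# A Random node picks among three equally weighted, looping wheel-rotation waves, so each playback holds one variant.
# Overridden attenuation uses logarithmic falloff with listener focus to lower volume when out of focus.
# Overridden concurrency allows a single instance and blocks new ones.
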SoundCue /Game/Sounds/Helios/POI/Mill/Spots/Spot_WaterWheel_Mill_Loop_Cue
root=Random_0
nodes [
 {id=Random_0, children [WavePlayer_1,WavePlayer_2,WavePlayer_3], properties {weights [1,1,1]}}
 {id=WavePlayer_1, properties {bLooping=true, soundWaveAssetPtr=&/Spot_Waterwheel_Rotate_01}}
 {id=WavePlayer_2, properties {bLooping=true, soundWaveAssetPtr=&/Spot_Waterwheel_Rotate_02}}
 {id=WavePlayer_3, properties {bLooping=true, soundWaveAssetPtr=&/Spot_Waterwheel_Rotate_03}}
]
properties {
 bOverrideAttenuation=true
 bOverrideConcurrency=true
 maxDistance=5400
 soundClassObject=&Fort_Ambient_Outdoors_Athena
 volumeMultiplier=0.626
 attenuationOverrides {
  bEnableListenerFocus=true
  distanceAlgorithm=Logarithmic
  falloffDistance=4500
  lPFRadiusMin=500
  nonFocusVolumeAttenuation=0.2
  attenuationShapeExtents=<900 90>
 }
 concurrencyOverrides {maxCount=1, resolutionRule=PreventNew}
}
