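# Looping ambient spot cue for the water sound of a rotating waterwheel at the Mill POI, in the outdoor ambience sound class.
# A Random node picks between two equally weighted looping water waves, played at lowered pitch and volume.
# Cue-level overrides apply logarithmic distance attenuation with no reverb send and limit concurrency to one instance, blocking new ones.
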SoundCue /Game/Sounds/Helios/POI/Mill/Spots/Spot_WaterWheel_Mill_Water_Loop_Cue
root=Random_1
nodes [
 {id=Random_1, children [WavePlayer_0,WavePlayer_4], properties {weights [1,1]}}
 {id=WavePlayer_0, properties {bLooping=true, soundWaveAssetPtr=&/Spot_Waterwheel_Rotate_Water_02}}
 {id=WavePlayer_4, properties {bLooping=true, soundWaveAssetPtr=&/Spot_Waterwheel_Rotate_Water_01}}
]
properties {
 bOverrideAttenuation=true
 bOverrideConcurrency=true
 maxDistance=5100
 pitchMultiplier=0.8
 soundClassObject=&Fort_Ambient_Outdoors_Athena
 volumeMultiplier=0.301
 attenuationOverrides {
  bEnableReverbSend=false
  distanceAlgorithm=Logarithmic
  falloffDistance=4500
  lPFRadiusMin=500
  stereoSpread=300
  attenuationShapeExtents=<600 90>
  customAttenuationCurve {editorCurveData {keys [<at 0 0>]}}
 }
 concurrencyOverrides {maxCount=1, resolutionRule=PreventNew}
}
